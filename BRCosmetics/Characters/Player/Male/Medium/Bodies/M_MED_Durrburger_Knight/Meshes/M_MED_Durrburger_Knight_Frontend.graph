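# Physics asset for the frontend variant of the Durrburger Knight character mesh, set up for physical motion of its six skirt panels.
# Capsules cover the pelvis, thighs and left/right front, side and back panels, with two angled capsules on each front panel.
# Each panel has limited swing and twist against a tiny master body, whose rotation is locked to the pelvis.
# Both thighs have limited angular constraints to the pelvis.
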PhysicsAsset /BRCosmetics/Characters/Player/Male/Medium/Bodies/M_MED_Durrburger_Knight/Meshes/M_MED_Durrburger_Knight_Frontend
bodies {
 dyn_skirt_l_01 [{shapeType=Capsule, length=12.0716, radius=1.7771, center=<6.9502 0.0005 0>, rotation=<rot 0 0 90>}]
 dyn_skirt_l_01_master [{shapeType=Capsule, length=0.505, radius=0.505}]
 dyn_skirt_l_back_01 [{shapeType=Capsule, length=11.6941, radius=2.6773, center=<6.9502 0.0005 0>, rotation=<rot 0 0 90>}]
 dyn_skirt_l_back_01_master [{shapeType=Capsule, length=0.505, radius=0.505}]
 dyn_skirt_l_front_01 [
  {shapeType=Capsule, length=15.2046, radius=3.5078, center=<12.6856 -9.9021 -0.3635>, rotation=<rot 0 0 76.8411>}
  {shapeType=Capsule, length=15.2637, radius=3.4487, center=<12.5801 2.9254 -0.8295>, rotation=<rot 0 0 112.3273>}
 ]
 dyn_skirt_l_front_master [{shapeType=Capsule, length=0.505, radius=0.505}]
 dyn_skirt_r_01 [{shapeType=Capsule, length=12.2864, radius=1.6606, center=<6.9502 0.0005 0>, rotation=<rot 0 0 90>}]
 dyn_skirt_r_01_master [{shapeType=Capsule, length=0.505, radius=0.505}]
 dyn_skirt_r_back_01 [{shapeType=Capsule, length=12.6774, radius=3.2095, center=<6.9502 0.0005 0>, rotation=<rot 0 0 90>}]
 dyn_skirt_r_back_01_master [{shapeType=Capsule, length=0.505, radius=0.505}]
 dyn_skirt_r_front_01 [
  {shapeType=Capsule, length=16.0846, radius=4.1159, center=<12.6856 8.0955 1.8629>, rotation=<rot 0 0 98.3521>}
  {shapeType=Capsule, length=16.0846, radius=4.1159, center=<12.739 -4.2407 0.4599>, rotation=<rot 0 0 63.582>}
 ]
 dyn_skirt_r_front_master [{shapeType=Capsule, length=0.505, radius=0.505}]
 pelvis [{shapeType=Capsule, length=9.1986, radius=10.2294, center=<-3.8341 0 0>}]
 thigh_l [{shapeType=Capsule, length=25.1141, radius=9.0519, center=<-20.5285 0.0018 0>, rotation=<rot 90 0 0>}]
 thigh_r [{shapeType=Capsule, length=21.3865, radius=9.8357, center=<19.7353 0.0014 0>, rotation=<rot 90 0 0>}]
}
constraints [
 {bone1Name=thigh_l, bone2Name=pelvis, swing1LimitDegrees=45, swing2LimitDegrees=45, twistLimitDegrees=45}
 {bone1Name=thigh_r, bone2Name=pelvis, swing1LimitDegrees=45, swing2LimitDegrees=45, twistLimitDegrees=45}
 {bone1Name=dyn_skirt_r_01, bone2Name=dyn_skirt_r_01_master, swing1LimitDegrees=45, swing2LimitDegrees=25, twistLimitDegrees=15}
 {bone1Name=dyn_skirt_r_front_01, bone2Name=dyn_skirt_r_front_master, swing1LimitDegrees=15, swing2LimitDegrees=90, twistLimitDegrees=15}
 {bone1Name=dyn_skirt_r_back_01, bone2Name=dyn_skirt_r_back_01_master, swing1LimitDegrees=45, swing2LimitDegrees=25, twistLimitDegrees=15}
 {bone1Name=dyn_skirt_l_01, bone2Name=dyn_skirt_l_01_master, swing1LimitDegrees=25, swing2LimitDegrees=25, twistLimitDegrees=15}
 {bone1Name=dyn_skirt_l_back_01, bone2Name=dyn_skirt_l_back_01_master, swing1LimitDegrees=45, swing2LimitDegrees=25, twistLimitDegrees=15}
 {bone1Name=dyn_skirt_l_front_01, bone2Name=dyn_skirt_l_front_master, swing1LimitDegrees=15, swing2LimitDegrees=60, twistLimitDegrees=25}
 {
  bone1Name=dyn_skirt_r_01_master
  bone2Name=pelvis
  swing1LimitDegrees=45
  swing1Motion=Locked
  swing2LimitDegrees=45
  swing2Motion=Locked
  twistLimitDegrees=45
  twistMotion=Locked
 }
 {
  bone1Name=dyn_skirt_r_front_master
  bone2Name=pelvis
  swing1LimitDegrees=45
  swing1Motion=Locked
  swing2LimitDegrees=45
  swing2Motion=Locked
  twistLimitDegrees=45
  twistMotion=Locked
 }
 {
  bone1Name=dyn_skirt_r_back_01_master
  bone2Name=pelvis
  swing1LimitDegrees=45
  swing1Motion=Locked
  swing2LimitDegrees=45
  swing2Motion=Locked
  twistLimitDegrees=45
  twistMotion=Locked
 }
 {
  bone1Name=dyn_skirt_l_01_master
  bone2Name=pelvis
  swing1LimitDegrees=45
  swing1Motion=Locked
  swing2LimitDegrees=45
  swing2Motion=Locked
  twistLimitDegrees=45
  twistMotion=Locked
 }
 {
  bone1Name=dyn_skirt_l_back_01_master
  bone2Name=pelvis
  swing1LimitDegrees=45
  swing1Motion=Locked
  swing2LimitDegrees=45
  swing2Motion=Locked
  twistLimitDegrees=45
  twistMotion=Locked
 }
 {
  bone1Name=dyn_skirt_l_front_master
  bone2Name=pelvis
  swing1LimitDegrees=45
  swing1Motion=Locked
  swing2LimitDegrees=45
  swing2Motion=Locked
  twistLimitDegrees=45
  twistMotion=Locked
 }
]
