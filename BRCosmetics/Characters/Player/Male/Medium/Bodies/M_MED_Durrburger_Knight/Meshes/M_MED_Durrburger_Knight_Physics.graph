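# Physics asset for the male medium Durrburger Knight player body, covering the pelvis, thighs and dynamic skirt bones.
# Six skirt chains (left and right front, side and back) each join a master bone to the pelvis and a child bone to that master.
# Front and side masters are angularly locked to the pelvis, while back masters and all children have limited motion.
# Capsules form every body, with two angled capsules on each front flap.
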PhysicsAsset /BRCosmetics/Characters/Player/Male/Medium/Bodies/M_MED_Durrburger_Knight/Meshes/M_MED_Durrburger_Knight_Physics
bodies {
 dyn_skirt_l_01 [{shapeType=Capsule, length=10.4829, radius=3.3659, center=<6.9502 0.0005 0>, rotation=<rot 0 0 90>}]
 dyn_skirt_l_01_master [{shapeType=Capsule, length=0.505, radius=0.505}]
 dyn_skirt_l_back_01 [{shapeType=Capsule, length=11.6994, radius=2.672, center=<6.9502 0.0005 1.595>, rotation=<rot 0 0 90>}]
 dyn_skirt_l_back_01_master [{shapeType=Capsule, length=10.536, radius=2.6993, center=<0 -0.0041 1.2996>, rotation=<rot 0 0 89.8201>}]
 dyn_skirt_l_front_01 [
  {shapeType=Capsule, length=15.565, radius=3.1474, center=<12.6857 -9.7212 -1.1374>, rotation=<rot 0 0 76.8413>}
  {shapeType=Capsule, length=15.6241, radius=3.0883, center=<12.5802 2.6234 -1.5647>, rotation=<rot 0 0 112.3274>}
 ]
 dyn_skirt_l_front_master [{shapeType=Capsule, length=0.505, radius=0.505}]
 dyn_skirt_r_01 [{shapeType=Capsule, length=10.6977, radius=3.2494, center=<6.9502 0.0005 0>, rotation=<rot 0 0 90>}]
 dyn_skirt_r_01_master [{shapeType=Capsule, length=0.505, radius=0.505}]
 dyn_skirt_r_back_01 [{shapeType=Capsule, length=12.6827, radius=3.2042, center=<6.9502 0.0005 1.595>, rotation=<rot 0 0 90>}]
 dyn_skirt_r_back_01_master [{shapeType=Capsule, length=10.536, radius=2.6993, center=<0 -0.0038 1.2996>, rotation=<rot 0 0 89.8322>}]
 dyn_skirt_r_front_01 [
  {shapeType=Capsule, length=17.6215, radius=2.579, center=<12.6857 7.9331 0.7569>, rotation=<rot 0 0 98.3521>}
  {shapeType=Capsule, length=17.6215, radius=2.579, center=<12.7391 -3.7433 -0.5413>, rotation=<rot 0 0 63.5819>}
 ]
 dyn_skirt_r_front_master [{shapeType=Capsule, length=0.505, radius=0.505}]
 pelvis [{shapeType=Capsule, length=9.0335, radius=10.3944, center=<-3.8341 -0.7828 0>}]
 thigh_l [{shapeType=Capsule, length=24.3547, radius=9.8114, center=<-20.5285 0.0018 0>, rotation=<rot 90 0 0>}]
 thigh_r [{shapeType=Capsule, length=21.3865, radius=9.8357, center=<19.7353 0.0014 0>, rotation=<rot 90 0 0>}]
}
constraints [
 {bone1Name=thigh_l, bone2Name=pelvis, swing1LimitDegrees=45, swing2LimitDegrees=45, twistLimitDegrees=45}
 {bone1Name=thigh_r, bone2Name=pelvis, swing1LimitDegrees=45, swing2LimitDegrees=45, twistLimitDegrees=45}
 {bone1Name=dyn_skirt_r_01, bone2Name=dyn_skirt_r_01_master, swing1LimitDegrees=45, swing2LimitDegrees=25, twistLimitDegrees=15}
 {bone1Name=dyn_skirt_r_front_01, bone2Name=dyn_skirt_r_front_master, swing1LimitDegrees=15, swing2LimitDegrees=90, twistLimitDegrees=15}
 {bone1Name=dyn_skirt_r_back_01, bone2Name=dyn_skirt_r_back_01_master, swing1LimitDegrees=5, swing2LimitDegrees=25, twistLimitDegrees=1}
 {bone1Name=dyn_skirt_l_01, bone2Name=dyn_skirt_l_01_master, swing1LimitDegrees=25, swing2LimitDegrees=25, twistLimitDegrees=15}
 {bone1Name=dyn_skirt_l_back_01, bone2Name=dyn_skirt_l_back_01_master, swing1LimitDegrees=5, swing2LimitDegrees=25, twistLimitDegrees=1}
 {bone1Name=dyn_skirt_l_front_01, bone2Name=dyn_skirt_l_front_master, swing1LimitDegrees=15, swing2LimitDegrees=60, twistLimitDegrees=25}
 {
  bone1Name=dyn_skirt_r_01_master
  bone2Name=pelvis
  swing1LimitDegrees=45
  swing1Motion=Locked
  swing2LimitDegrees=45
  swing2Motion=Locked
  twistLimitDegrees=45
  twistMotion=Locked
 }
 {
  bone1Name=dyn_skirt_r_front_master
  bone2Name=pelvis
  swing1LimitDegrees=45
  swing1Motion=Locked
  swing2LimitDegrees=45
  swing2Motion=Locked
  twistLimitDegrees=45
  twistMotion=Locked
 }
 {bone1Name=dyn_skirt_r_back_01_master, bone2Name=pelvis, swing1LimitDegrees=5, swing2LimitDegrees=25, twistLimitDegrees=1}
 {
  bone1Name=dyn_skirt_l_01_master
  bone2Name=pelvis
  swing1LimitDegrees=45
  swing1Motion=Locked
  swing2LimitDegrees=45
  swing2Motion=Locked
  twistLimitDegrees=45
  twistMotion=Locked
 }
 {bone1Name=dyn_skirt_l_back_01_master, bone2Name=pelvis, swing1LimitDegrees=5, swing2LimitDegrees=25, twistLimitDegrees=1}
 {
  bone1Name=dyn_skirt_l_front_master
  bone2Name=pelvis
  swing1LimitDegrees=45
  swing1Motion=Locked
  swing2LimitDegrees=45
  swing2Motion=Locked
  twistLimitDegrees=45
  twistMotion=Locked
 }
]
