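# Animation Blueprint for the Durrburger Knight player body that adds secondary motion to its skirt, shoulder pads and pouches.
# It copies the parent mesh pose, then blends crouch, jump/fall, front-end and default branches, each using thigh constraints, bone offsets and RigidBody simulation, with Control Rig when crouching.
# The event graph derives corrective alphas from limb angles, movement and pistol stance, and resets dynamics after crouching, emotes or skydiving.
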AnimBlueprint /BRCosmetics/Characters/Player/Male/Medium/Bodies/M_MED_Durrburger_Knight/Meshes/M_MED_Durrburger_Knight_AnimBP
target_skeleton=&/M_MED_Durrburger_Knight_Skeleton
graphs [
 AnimGraph {nodes [
   Component To Local()
   n1 = "Transform (Modify) Bone - Bone: dyn_skirt_r_back_01_master"(Alpha=1, Translation="-1,-2.5,1", Rotation="0,0,0", Scale="1,1,1") {
    rotationMode=BMM_Additive
    scaleMode=BMM_Additive
    translationMode=BMM_Additive
    alphaBoolBlend {blendOption=Linear}
    boneToModify {boneName=dyn_skirt_r_back_01_master}
    translation=<-1 -2.5 1>
   }
   "Transform (Modify) Bone - Bone: dyn_skirt_l_back_01_master"(ComponentPose=@n1, Alpha=1, Translation="1,-2.5,1", Rotation="0,0,0", Scale="1,1,1") {
    rotationMode=BMM_Additive
    scaleMode=BMM_Additive
    translationMode=BMM_Additive
    alphaBoolBlend {blendOption=Linear}
    boneToModify {boneName=dyn_skirt_l_back_01_master}
    translation=<1 -2.5 1>
   }
   n2 = "Transform (Modify) Bone - Bone: dyn_skirt_r_back_01_master"(Alpha=1, Translation="0,-2.5,1", Rotation="0,0,0", Scale="1,1,1") {
    rotationMode=BMM_Additive
    scaleMode=BMM_Additive
    translationMode=BMM_Additive
    alphaBoolBlend {blendOption=Linear}
    boneToModify {boneName=dyn_skirt_r_back_01_master}
    translation {y=-2.5, z=1}
   }
   "Transform (Modify) Bone - Bone: dyn_skirt_l_back_01_master"(ComponentPose=@n2, Alpha=1, Translation="0,-2.5,1", Rotation="0,0,0", Scale="1,1,1") {
    rotationMode=BMM_Additive
    scaleMode=BMM_Additive
    translationMode=BMM_Additive
    alphaBoolBlend {blendOption=Linear}
    boneToModify {boneName=dyn_skirt_l_back_01_master}
    translation {y=-2.5, z=1}
   }
   RigidBody() {
    bClampLinearTranslationLimitToRefPose=true
    bOverrideWorldGravity=true
    overridePhysicsAsset=&/M_MED_Durrburger_Knight_Physics
    simulationSpace=BaseBoneSpace
    alphaBoolBlend {blendOption=Linear}
    baseBoneRef {boneName=root}
    componentAppliedLinearAccClamp=<1500 1500 1500>
    componentLinearAccScale=<0.4 0.4 0.5>
    overrideWorldGravity {z=-1100}
   }
   n3 = "Transform (Modify) Bone - Bone: dyn_l_shoulderpad"(
    ComponentPose=Local To Component(LocalPose=Use cached pose 'Pre_Jump'())
    Alpha=Get pistolAtRest() -> (pistolAtRest:double)
    Translation="0,0,0"
    Rotation="0,-10,0"
    Scale="1,1,1"
   ) {
    rotationMode=BMM_Additive
    scaleMode=BMM_Additive
    translationMode=BMM_Additive
    alphaBoolBlend {blendOption=Linear}
    boneToModify {boneName=dyn_l_shoulderpad}
    rotation {yaw=-10}
   }
   n4 = "Transform (Modify) Bone - Bone: dyn_skirt_l_back_01_master"(ComponentPose=@n3, Alpha=1, Translation="0,-0.5,0", Rotation="0,0,0", Scale="1,1,1") {
    rotationMode=BMM_Additive
    scaleMode=BMM_Additive
    translationMode=BMM_Additive
    alphaBoolBlend {blendOption=Linear}
    boneToModify {boneName=dyn_skirt_l_back_01_master}
    translation {y=-0.5}
   }
   n5 = "Transform (Modify) Bone - Bone: dyn_skirt_r_back_01_master"(ComponentPose=@n4, Alpha=1, Translation="0,-0.5,0", Rotation="0,0,0", Scale="1,1,1") {
    rotationMode=BMM_Additive
    scaleMode=BMM_Additive
    translationMode=BMM_Additive
    alphaBoolBlend {blendOption=Linear}
    boneToModify {boneName=dyn_skirt_r_back_01_master}
    translation {y=-0.5}
   }
   Output Pose(Result=Component To Local(ComponentPose=@n5)) {layerGroup=None}
   n6 = "Constraint - dyn_skirt_r_front_01 (P)"(ComponentPose=Local To Component(LocalPose=Copy Pose From Mesh() {bUseAttachedParent=true}), Alpha=1, ConstraintWeights_0=1) {
    alphaBoolBlend {blendOption=Linear}
    boneToModify {boneName=dyn_skirt_r_front_01}
    constraintSetup [{targetBone {boneName=thigh_twist_01_r}}]
    constraintWeights [1]
   }
   n7 = "Constraint - dyn_skirt_l_front_01 (P)"(ComponentPose=@n6, Alpha=1, ConstraintWeights_0=1) {
    alphaBoolBlend {blendOption=Linear}
    boneToModify {boneName=dyn_skirt_l_front_01}
    constraintSetup [{targetBone {boneName=thigh_twist_01_l}}]
    constraintWeights [1]
   }
   n8 = "Transform (Modify) Bone - Bone: dyn_skirt_r_front_master"(ComponentPose=@n7, Alpha=1, Translation="0,1,2", Rotation="0,0,0", Scale="1,1,1") {
    rotationMode=BMM_Additive
    scaleMode=BMM_Additive
    translationMode=BMM_Additive
    alphaBoolBlend {blendOption=Linear}
    boneToModify {boneName=dyn_skirt_r_front_master}
    translation {y=1, z=2}
   }
   n9 = "Transform (Modify) Bone - Bone: dyn_skirt_l_front_master"(ComponentPose=@n8, Alpha=1, Translation="0,1,2", Rotation="0,0,0", Scale="1,1,1") {
    rotationMode=BMM_Additive
    scaleMode=BMM_Additive
    translationMode=BMM_Additive
    alphaBoolBlend {blendOption=Linear}
    boneToModify {boneName=dyn_skirt_l_front_master}
    translation {y=1, z=2}
   }
   n10 = RigidBody(ComponentPose=@n9) {
    bClampLinearTranslationLimitToRefPose=true
    bOverrideWorldGravity=true
    overridePhysicsAsset=&M_MED_Durrburger_Knight_Physics
    simulationSpace=BaseBoneSpace
    alphaBoolBlend {blendOption=Linear}
    baseBoneRef {boneName=pelvis}
    componentAppliedLinearAccClamp=<1500 1500 1500>
    componentLinearAccScale=<0.4 0.4 0.5>
    overrideWorldGravity {z=-1100}
    simSpaceSettings {maxLinearAcceleration=100, maxLinearVelocity=100}
   }
   n11 = "Constraint - dyn_skirt_r_front_01 (P)"(ComponentPose=Local To Component(LocalPose=Copy Pose From Mesh() {bUseAttachedParent=true}), Alpha=1, ConstraintWeights_0=1) {
    alphaBoolBlend {blendOption=Linear}
    boneToModify {boneName=dyn_skirt_r_front_01}
    constraintSetup [{targetBone {boneName=thigh_twist_01_r}}]
    constraintWeights [1]
   }
   n12 = "Constraint - dyn_skirt_l_front_01 (P)"(ComponentPose=@n11, Alpha=1, ConstraintWeights_0=1) {
    alphaBoolBlend {blendOption=Linear}
    boneToModify {boneName=dyn_skirt_l_front_01}
    constraintSetup [{targetBone {boneName=thigh_twist_01_l}}]
    constraintWeights [1]
   }
   n13 = "Transform (Modify) Bone - Bone: dyn_skirt_r_front_master"(ComponentPose=@n12, Alpha=1, Translation="0,1,2", Rotation="0,0,0", Scale="1,1,1") {
    rotationMode=BMM_Additive
    scaleMode=BMM_Additive
    translationMode=BMM_Additive
    alphaBoolBlend {blendOption=Linear}
    boneToModify {boneName=dyn_skirt_r_front_master}
    translation {y=1, z=2}
   }
   n14 = "Transform (Modify) Bone - Bone: dyn_skirt_l_front_master"(ComponentPose=@n13, Alpha=1, Translation="0,1,2", Rotation="0,0,0", Scale="1,1,1") {
    rotationMode=BMM_Additive
    scaleMode=BMM_Additive
    translationMode=BMM_Additive
    alphaBoolBlend {blendOption=Linear}
    boneToModify {boneName=dyn_skirt_l_front_master}
    translation {y=1, z=2}
   }
   n15 = RigidBody(ComponentPose=@n14) {
    bClampLinearTranslationLimitToRefPose=true
    bOverrideWorldGravity=true
    overridePhysicsAsset=&/M_MED_Durrburger_Knight_Frontend
    simulationSpace=BaseBoneSpace
    alphaBoolBlend {blendOption=Linear}
    baseBoneRef {boneName=root}
    componentAppliedLinearAccClamp=<1500 1500 1500>
    componentLinearAccScale=<0.4 0.4 0.5>
    overrideWorldGravity {z=-1100}
   }
   n16 = "Constraint - dyn_skirt_r_front_01 (P)"(ComponentPose=Local To Component(LocalPose=Copy Pose From Mesh() {bUseAttachedParent=true}), Alpha=1, ConstraintWeights_0=1) {
    alphaBoolBlend {blendOption=Linear}
    boneToModify {boneName=dyn_skirt_r_front_01}
    constraintSetup [{targetBone {boneName=thigh_twist_01_r}}]
    constraintWeights [1]
   }
   n17 = "Constraint - dyn_skirt_l_front_01 (P)"(ComponentPose=@n16, Alpha=1, ConstraintWeights_0=1) {
    alphaBoolBlend {blendOption=Linear}
    boneToModify {boneName=dyn_skirt_l_front_01}
    constraintSetup [{targetBone {boneName=thigh_twist_01_l}}]
    constraintWeights [1]
   }
   n18 = "Transform (Modify) Bone - Bone: dyn_skirt_r_front_master"(ComponentPose=@n17, Alpha=1, Translation="0,1,2", Rotation="0,0,0", Scale="1,1,1") {
    rotationMode=BMM_Additive
    scaleMode=BMM_Additive
    translationMode=BMM_Additive
    alphaBoolBlend {blendOption=Linear}
    boneToModify {boneName=dyn_skirt_r_front_master}
    translation {y=1, z=2}
   }
   n19 = "Transform (Modify) Bone - Bone: dyn_skirt_l_front_master"(ComponentPose=@n18, Alpha=1, Translation="0,1,2", Rotation="0,0,0", Scale="1,1,1") {
    rotationMode=BMM_Additive
    scaleMode=BMM_Additive
    translationMode=BMM_Additive
    alphaBoolBlend {blendOption=Linear}
    boneToModify {boneName=dyn_skirt_l_front_master}
    translation {y=1, z=2}
   }
   n20 = Get leg_raise() -> (leg_raise:double)
   n21 = "Transform (Modify) Bone - Bone: dyn_skirt_l_front_master"(ComponentPose=@n19, Alpha=@n20, Translation="0,2,4", Rotation="0,0,0", Scale="1,1,1") {
    rotationMode=BMM_Additive
    scaleMode=BMM_Additive
    translationMode=BMM_Additive
    alphaBoolBlend {blendOption=Linear}
    boneToModify {boneName=dyn_skirt_l_front_master}
    translation {y=2, z=4}
   }
   n22 = "Transform (Modify) Bone - Bone: dyn_skirt_r_01_master"(ComponentPose=@n21, Alpha=@n20, Translation="0,0,4", Rotation="0,0,0", Scale="1,1,1") {
    rotationMode=BMM_Additive
    scaleMode=BMM_Additive
    translationMode=BMM_Additive
    alphaBoolBlend {blendOption=Linear}
    boneToModify {boneName=dyn_skirt_r_01_master}
    translation {z=4}
   }
   n23 = RigidBody(ComponentPose=@n22) {
    bClampLinearTranslationLimitToRefPose=true
    bOverrideWorldGravity=true
    overridePhysicsAsset=&M_MED_Durrburger_Knight_Physics
    simulationSpace=BaseBoneSpace
    alphaBoolBlend {blendOption=Linear}
    baseBoneRef {boneName=pelvis}
    componentAppliedLinearAccClamp=<1500 1500 1500>
    componentLinearAccScale=<0.4 0.4 0.5>
    overrideWorldGravity {z=-1100}
   }
   n24 = Get leg_raise()
   n25 = "Transform (Modify) Bone - Bone: dyn_skirt_r_front_master"(ComponentPose=@n23, Alpha=@n24, Translation="0,2,8", Rotation="0,0,-15", Scale="1,1,1") {
    rotationMode=BMM_Additive
    scaleMode=BMM_Additive
    translationMode=BMM_Additive
    alphaBoolBlend {blendOption=Linear}
    boneToModify {boneName=dyn_skirt_r_front_master}
    rotation {roll=-15}
    translation {y=2, z=8}
   }
   n26 = "Transform (Modify) Bone - Bone: dyn_skirt_l_front_master"(ComponentPose=@n25, Alpha=@n24, Translation="0,2,3", Rotation="0,0,-35", Scale="1,1,1") {
    rotationMode=BMM_Additive
    scaleMode=BMM_Additive
    translationMode=BMM_Additive
    alphaBoolBlend {blendOption=Linear}
    boneToModify {boneName=dyn_skirt_l_front_master}
    rotation {roll=-35}
    translation {y=2, z=3}
   }
   n27 = "Transform (Modify) Bone - Bone: dyn_skirt_r_back_01_master"(ComponentPose=@n26, Alpha=@n24, Translation="0,0,1", Rotation="-20,0,0", Scale="1,1,1") {
    rotationMode=BMM_Additive
    scaleMode=BMM_Additive
    translationMode=BMM_Additive
    alphaBoolBlend {blendOption=Linear}
    boneToModify {boneName=dyn_skirt_r_back_01_master}
    rotation {pitch=-20}
    translation {z=1}
   }
   n28 = Get is_jumping() -> (is_jumping:double)
   n29 = "Transform (Modify) Bone - Bone: dyn_skirt_r_01_master"(ComponentPose=@n27, Alpha=@n28, Translation="0,-5,10", Rotation="0,0,-30", Scale="1,1,1") {
    rotationMode=BMM_Additive
    scaleMode=BMM_Additive
    translationMode=BMM_Additive
    alphaBoolBlend {blendOption=Linear}
    boneToModify {boneName=dyn_skirt_r_01_master}
    rotation {roll=-30}
    translation {y=-5, z=10}
   }
   n30 = "Transform (Modify) Bone - Bone: dyn_skirt_r_front_master"(ComponentPose=@n29, Alpha=@n28, Translation="0,0,2", Rotation="0,0,-15", Scale="1,1,1") {
    rotationMode=BMM_Additive
    scaleMode=BMM_Additive
    translationMode=BMM_Additive
    alphaBoolBlend {blendOption=Linear}
    boneToModify {boneName=dyn_skirt_r_front_master}
    rotation {roll=-15}
    translation {z=2}
   }
   n31 = Blend Poses by bool(
    BlendPose_0=Component To Local(ComponentPose=@n30)
    BlendPose_1=Blend Poses by bool(BlendPose_0=Component To Local(ComponentPose=@n15), BlendPose_1=Component To Local(ComponentPose=@n10), bActiveValue=Get bIsInFrontEnd()) {blendType=Linear}
    bActiveValue=OR Boolean(A=Get bIsJumping(), B=Get bIsFalling())
   ) {blendType=Linear}
   n32 = "Constraint - dyn_skirt_r_front_01 (P)"(ComponentPose=Local To Component(LocalPose=Copy Pose From Mesh() {bUseAttachedParent=true}), Alpha=1, ConstraintWeights_0=1) {
    alphaBoolBlend {blendOption=Linear}
    boneToModify {boneName=dyn_skirt_r_front_01}
    constraintSetup [{targetBone {boneName=thigh_twist_01_r}}]
    constraintWeights [1]
   }
   n33 = "Constraint - dyn_skirt_l_front_01 (P)"(ComponentPose=@n32, Alpha=1, ConstraintWeights_0=1) {
    alphaBoolBlend {blendOption=Linear}
    boneToModify {boneName=dyn_skirt_l_front_01}
    constraintSetup [{targetBone {boneName=thigh_twist_01_l}}]
    constraintWeights [1]
   }
   n34 = "Transform (Modify) Bone - Bone: dyn_skirt_l_front_master"(ComponentPose=@n33, Alpha=1, Translation="0,1,1", Rotation="0,0,0", Scale="1,1,1") {
    rotationMode=BMM_Additive
    scaleMode=BMM_Additive
    translationMode=BMM_Additive
    alphaBoolBlend {blendOption=Linear}
    boneToModify {boneName=dyn_skirt_l_front_master}
    translation {y=1, z=1}
   }
   n35 = "Transform (Modify) Bone - Bone: dyn_skirt_r_front_master"(ComponentPose=@n34, Alpha=1, Translation="0,2,1", Rotation="0,0,0", Scale="1,1,1") {
    rotationMode=BMM_Additive
    scaleMode=BMM_Additive
    translationMode=BMM_Additive
    alphaBoolBlend {blendOption=Linear}
    boneToModify {boneName=dyn_skirt_r_front_master}
    translation {y=2, z=1}
   }
   n36 = RigidBody(ComponentPose=@n35) {
    bClampLinearTranslationLimitToRefPose=true
    bOverrideWorldGravity=true
    overridePhysicsAsset=&M_MED_Durrburger_Knight_Physics
    simulationSpace=BaseBoneSpace
    alphaBoolBlend {blendOption=Linear}
    baseBoneRef {boneName=pelvis}
    componentAppliedLinearAccClamp=<1500 1500 1500>
    componentLinearAccScale=<0.4 0.4 0.5>
    overrideWorldGravity {z=-1100}
   }
   n37 = Control Rig(Source=Component To Local(ComponentPose=@n36)) {
    alphaBoolBlend {blendOption=Linear}
    controlRigAssetReference {blueprintRigClass=&/M_MED_Durrburger_Knight_CtrlRig.M_MED_Durrburger_Knight_CtrlRig_C}
   }
   n38 = Component To Local(ComponentPose=Local To Component(LocalPose=@n37))
   n39 = Get bIsCrouching()
   n40 = "Transform (Modify) Bone - Bone: dyn_shoulderpad_rotate_r"(
    ComponentPose=Local To Component(LocalPose=Blend Poses by bool(BlendPose_0=@n38, BlendPose_1=@n31, bActiveValue=@n39) {blendType=Linear})
    Alpha=Get aimUP() -> (aimUP:double)
    Translation="0,0,0"
    Rotation="10,10,0"
    Scale="1,1,1"
   ) {
    rotationMode=BMM_Additive
    translationMode=BMM_Additive
    alphaBoolBlend {blendOption=Linear}
    boneToModify {boneName=dyn_shoulderpad_rotate_r}
    rotation {pitch=10, yaw=10}
   }
   n41 = Get isCrouching() -> (isCrouching:double)
   n42 = "Transform (Modify) Bone - Bone: dyn_pouch_01_master"(ComponentPose=@n40, Alpha=@n41, Translation="1,-1,0", Rotation="15,0,0", Scale="1,1,1") {
    rotationMode=BMM_Additive
    translationMode=BMM_Additive
    alphaBoolBlend {blendOption=Linear}
    boneToModify {boneName=dyn_pouch_01_master}
    rotation {pitch=15}
    translation=<1 -1>
   }
   n43 = "Transform (Modify) Bone - Bone: dyn_pouch_02_master"(ComponentPose=@n42, Alpha=@n41, Translation="0,0,0", Rotation="7,0,0", Scale="1,1,1") {
    rotationMode=BMM_Additive
    translationMode=BMM_Additive
    alphaBoolBlend {blendOption=Linear}
    boneToModify {boneName=dyn_pouch_02_master}
    rotation {pitch=7}
   }
   Save cached pose 'Pre_Jump'(Pose=Component To Local(ComponentPose=@n43)) {cacheName=Pre_Jump}
  ]}
 EventGraph {nodes [
   n1 = GetOwnerSkeletalMeshComponent()
   n2 = BreakRotator(InRot=BreakTransform(InTransform=GetDeltaTransformFromRefPose(self=@n1, BoneName=thigh_r, BaseName=pelvis)).Rotation)
   n3 = PrintString!(execute, InString="To String (Float)"(InDouble=@n2.Yaw), bPrintToScreen, bPrintToLog, TextColor, Duration, Key) -> (then)
   n4 = Get bIsJumping()
   n5 = Get bIsFalling()
   Set is_jumping(execute=[@n3.then, @n14.then], is_jumping:double=SelectFloat(A=1, B=0, bPickA=OR Boolean(A=@n4, B=@n5))) -> (Output_Get:double)
   Event BlueprintUpdateAnimation()
   > n6 = Sequence!(execute) -> (then_0, then_1, then_2:Exec)
   > n7 = Branch(Condition=Get bIsCrouching())
   > Set isCrouching(isCrouching:double=1) -> (Output_Get:double)
   Set isCrouching(execute=@n7.else, isCrouching=0)
   n8 = BreakTransform(InTransform=GetDeltaTransformFromRefPose(self=GetOwnerSkeletalMeshComponent(), BoneName=upperarm_r, BaseName=clavicle_r))
   Set aimUP(execute=@n6.then_1, aimUP:double=MapRangeClamped(Value=BreakRotator(InRot=@n8.Rotation).Yaw, InRangeA=95, InRangeB=120)) -> (Output_Get:double)
   > n9 = Cast To FortPlayerAnimInstance(Object=GetAnimInstance(self=GetOwnerSkeletalMeshComponent()))
   > Set pistolAtRest(pistolAtRest:double=0) -> (Output_Get:double)
   > n10 = Switch on EFortWeaponCoreAnimation!(execute, Selection:EFortWeaponCoreAnimation Enum=Get WeaponCoreAnim(self=@n9)) -> (
    Melee:Exec
    Pistol:Exec
    Shotgun:Exec
    PaperBlueprint:Exec
    Rifle:Exec
    MeleeOneHand:Exec
    MachinePistol:Exec
    RocketLauncher:Exec
    GrenadeLauncher:Exec
    GoingCommando:Exec
    AssaultRifle:Exec
    TacticalShotgun:Exec
    SniperRifle:Exec
    TrapPlacement:Exec
    ShoulderLauncher:Exec
    AbilityDecoTool:Exec
    Crossbow:Exec
    C4:Exec
    RemoteControl:Exec
    DualWield:Exec
    AR_BullPup:Exec
    AR_ForwardGrip:Exec
    MedPackPaddles:Exec
    SMG_P90:Exec
    AR_DrumGun:Exec
    Consumable_Small:Exec
    Consumable_Large:Exec
    Balloon:Exec
    MountedTurret:Exec
    CreativeTool:Exec
    ExplosiveBow:Exec
    AshtonIndigo:Exec
    AshtonChicago:Exec
    MeleeDualWield:Exec
    Unarmed:Exec
   )
   n11 = Branch(execute=@n10.Pistol, Condition=Get bIsTargeting())
   Set pistolAtRest(execute=[@n11.then, @n13.else], pistolAtRest=0)
   n12 = Branch(execute=@n11.else, Condition=Get bIsRelaxedLevel1() -> (bIsRelaxedLevel1:bool))
   Set pistolAtRest(execute=[@n12.then, @n13.then], pistolAtRest=1)
   n13 = Branch(execute=@n12.else, Condition=Get bIsRelaxedLevel2() -> (bIsRelaxedLevel2:bool))
   n14 = Set leg_raise(execute=@n6.then_2, leg_raise:double=MapRangeClamped(Value=@n2.Yaw, InRangeA=-40, InRangeB=-80)) -> (Output_Get:double)
   Event OnEndCrouching()
   > n15 = Delay(Duration=0.15)
   ResetDynamics(execute=[@n17.then, @n15.then], InTeleportType=TeleportPhysics)
   > ResetDynamics(InTeleportType=ResetPhysics)
   n16 = Event OnEndPlayingEmote() {is=title K2Node_Event}
   n17 = Delay(execute=[@n16.then, @n18.then], Duration=0.1)
   n18 = Event OnEndSkydiving()
   Event OnBeginCrouching()
  ]}
]
variables [
 aimUP {type=number}
 is_jumping {type=number}
 isCrouching {type=number}
 leg_raise {type=number}
 pistolAtRest {type=number}
]
