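# Control Rig for the Durrburger Knight's face accessory mesh that adjusts the head accessory and cheese bones as the head turns relative to spine_05.
# Clamped remaps of the head's Euler Z angle offset and rotate dyn_head.
# The resulting dyn_head angle then drives corrective transforms on the back-left, front-right and front-left cheese bones.
# Two further modify nodes, for dyn_head and front-right cheese, sit outside the execution chain.
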ControlRigBlueprint /BRCosmetics/Characters/Player/Male/Medium/Bodies/M_MED_Durrburger_Knight/Meshes/Parts/M_MED_Durrburger_Knight_FaceAcc_CtrlRig
graphs [
 RigVMFunctionLibrary {links [], nodes []}
 RigVMModel {nodes [
   RigUnit_BeginExecution = &RigUnit_BeginExecution()
   n1 = &RigVMFunction_MathQuaternionToEuler(
    RotationOrder=ZYX
    Value=&RigUnit_GetRelativeTransformForItem(Child=(Type=Bone,Name="head"), Parent=(Type=Bone,Name="spine_05")).RelativeTransform.Rotation
   )
   &RigUnit_ModifyTransforms(
    ItemToModify=((Item=(Type=Bone,Name="dyn_head"),Transform=(Rotation=(X=0,Y=0,Z=0,W=1),Translation=(X=0,Y=0,Z=-0.5),Scale3D=(X=1,Y=1,Z=1))))
    Weight=1
    ItemToModify.0.Transform.Translation.Y=&RigVMFunction_MathFloatRemap(SourceMaximum=17, SourceMinimum=-20, TargetMaximum=-1, TargetMinimum=0, bClamp=true, Value=@n1.Result.Z).Result
    ItemToModify.0.Transform.Translation.Z=&RigVMFunction_MathFloatRemap(SourceMaximum=17, SourceMinimum=-20, TargetMaximum=-0.5, TargetMinimum=0, bClamp=true, Value=@n1.Result.Z).Result
    ItemToModify.0.Transform.Rotation=&RigVMFunction_MathQuaternionFromEuler(
     Euler=(X=0,Y=0,Z=0)
     Euler.Z=&RigVMFunction_MathFloatRemap(SourceMaximum=17, SourceMinimum=-20, TargetMaximum=-3, TargetMinimum=30, bClamp=true, Value=@n1.Result.Z).Result
    ).Result
    ExecutePin=@RigUnit_BeginExecution.ExecutePin
   )
   n2 = &RigVMFunction_MathQuaternionToEuler(
    RotationOrder=ZYX
    Value=&RigUnit_GetRelativeTransformForItem(Child=(Type=Bone,Name="dyn_head"), Parent=(Type=Bone,Name="spine_05")).RelativeTransform.Rotation
   )
   > &RigUnit_ModifyTransforms(
    ItemToModify=((Item=(Type=Bone,Name="dyn_cheese_bk_l"),Transform=(Rotation=(X=0,Y=0,Z=0,W=1),Translation=(X=0,Y=0,Z=0),Scale3D=(X=1,Y=1,Z=1))))
    Weight=1
    ItemToModify.0.Transform.Translation.Z=&RigVMFunction_MathFloatRemap(SourceMaximum=20, SourceMinimum=0, TargetMaximum=0, TargetMinimum=-6, bClamp=true, Value=@n2.Result.Z).Result
    ItemToModify.0.Transform.Rotation=&RigVMFunction_MathQuaternionFromEuler(
     Euler=(X=0,Y=0,Z=40)
     Euler.Y=&RigVMFunction_MathFloatRemap(SourceMaximum=20, SourceMinimum=0, TargetMaximum=20, TargetMinimum=-30, bClamp=true, Value=@n2.Result.Z).Result
     Euler.Z=&RigVMFunction_MathFloatRemap(SourceMaximum=20, SourceMinimum=0, TargetMaximum=0, TargetMinimum=40, bClamp=true, Value=@n2.Result.Z).Result
    ).Result
   )
   n3 = &RigVMFunction_MathQuaternionToEuler(
    RotationOrder=ZYX
    Value=&RigUnit_GetRelativeTransformForItem(Child=(Type=Bone,Name="dyn_head"), Parent=(Type=Bone,Name="spine_05")).RelativeTransform.Rotation
   )
   > &RigUnit_ModifyTransforms(
    ItemToModify=((Item=(Type=Bone,Name="dyn_cheese_fr_r"),Transform=(Rotation=(X=0,Y=0,Z=0,W=1),Translation=(X=0,Y=0,Z=0),Scale3D=(X=1,Y=1,Z=1))))
    Weight=1
    ItemToModify.0.Transform.Translation.X=&RigVMFunction_MathFloatRemap(SourceMaximum=20, SourceMinimum=0, TargetMaximum=-1, TargetMinimum=0, bClamp=true, Value=@n3.Result.Z).Result
    ItemToModify.0.Transform.Rotation=&RigVMFunction_MathQuaternionFromEuler(
     Euler=(X=0,Y=0,Z=0)
     Euler.Y=&RigVMFunction_MathFloatRemap(SourceMaximum=20, SourceMinimum=0, TargetMaximum=-40, TargetMinimum=0, bClamp=true, Value=@n3.Result.Z).Result
    ).Result
   )
   n4 = &RigVMFunction_MathQuaternionToEuler(
    RotationOrder=ZYX
    Value=&RigUnit_GetRelativeTransformForItem(Child=(Type=Bone,Name="dyn_head"), Parent=(Type=Bone,Name="spine_05")).RelativeTransform.Rotation
   )
   > &RigUnit_ModifyTransforms(
    ItemToModify=((Item=(Type=Bone,Name="dyn_cheese_fr_l"),Transform=(Rotation=(X=0,Y=0,Z=0,W=1),Translation=(X=0,Y=0,Z=0),Scale3D=(X=1,Y=1,Z=1))))
    Mode=OverrideLocal
    Weight=1
    ItemToModify.0.Transform.Translation.X=&RigVMFunction_MathFloatRemap(SourceMaximum=20, SourceMinimum=0, TargetMaximum=-0.5, TargetMinimum=0, bClamp=true, Value=@n4.Result.Z).Result
    ItemToModify.0.Transform.Rotation=&RigVMFunction_MathQuaternionFromEuler(
     Euler=(X=0,Y=0,Z=0)
     Euler.Y=&RigVMFunction_MathFloatRemap(SourceMaximum=20, SourceMinimum=0, TargetMaximum=-40, TargetMinimum=0, bClamp=true, Value=@n4.Result.Z).Result
    ).Result
   )
   &RigUnit_ModifyTransforms(
    ExecutePin=()
    ItemToModify=((Item=(Type=Bone,Name="dyn_head"),Transform=(Rotation=(X=0,Y=0,Z=0,W=1),Translation=(X=0,Y=0,Z=0),Scale3D=(X=1,Y=1,Z=1))))
    Weight=1
    ItemToModify.0.Transform.Rotation=&RigVMFunction_MathQuaternionFromEuler(
     Euler=(X=0,Y=0,Z=0)
     Euler.Y=&RigVMFunction_MathFloatRemap(SourceMaximum=-12, SourceMinimum=-5, TargetMaximum=0, TargetMinimum=0, bClamp=true, Value=@n1.Result.Y).Result
    ).Result
   )
   &RigUnit_ModifyTransforms(
    ExecutePin=()
    ItemToModify=((Item=(Type=Bone,Name="dyn_cheese_fr_r"),Transform=(Rotation=(X=0,Y=0,Z=0,W=1),Translation=(X=0,Y=0,Z=0),Scale3D=(X=1,Y=1,Z=1))))
    Mode=AdditiveGlobal
    Weight=1
    ItemToModify.0.Transform.Translation.Z=&RigVMFunction_MathFloatRemap(SourceMaximum=20, SourceMinimum=0, TargetMinimum=0, bClamp=true, Value=@n3.Result.Z).Result
   )
  ]}
]
hierarchy [
 root {loc=<0 0 0>, rot=<0 0 0 1>}
 pelvis {parent=root, loc=<0 -0.7071 93.1191>, rot=<0.0127 0.707 -0.0127 -0.707>}
 spine_01 {parent=pelvis, loc=<6.7494 -0.6388 0>, rot=<0 0 0.034 -0.9994>}
 spine_02 {parent=spine_01, loc=<7.2518 -0.035 0>, rot=<0 0 0.0238 -0.9997>}
 spine_03 {parent=spine_02, loc=<6.0104 0 0>, rot=<0 0 0.0149 -0.9999>}
 spine_04 {parent=spine_03, loc=<8.1024 0 0>, rot=<0 0 -0.0132 -0.9999>}
 spine_05 {parent=spine_04, loc=<10.0146 0 0>, rot=<0 0 -0.0015 -1>}
 clavicle_l {parent=spine_05, loc=<4.3575 -0.1466 -2.3223>, rot=<0.0465 -0.7289 -0.009 -0.6829>}
 upperarm_l {parent=clavicle_l, loc=<15.5111 0 0>, rot=<0.0226 -0.3234 -0.0088 -0.9459>}
 clavicle_r {parent=spine_05, loc=<4.3575 -0.1466 2.3223>, rot=<0.7289 0.0465 -0.6829 0.009>}
 upperarm_r {parent=clavicle_r, loc=<-15.5111 0 0>, rot=<0.0226 -0.3234 -0.0088 -0.9459>}
 neck_01 {parent=spine_05, loc=<11.3138 0.1529 0>, rot=<0 0 -0.2003 -0.9797>}
 neck_02 {parent=neck_01, loc=<4.5627 0 0>, rot=<0 0 0.0504 -0.9987>}
 head {parent=neck_02, loc=<5.3914 0 0>, rot=<0 0 0.115 -0.9934>}
 dyn_head {parent=neck_02, loc=<5.3914 0 0>, rot=<0 0 0.115 -0.9934>}
 dyn_toothpick_base_master {parent=dyn_head, loc=<11.6056 1.9276 0.0833>, rot=<0.7282 -0.0545 0.0526 -0.6812>}
 dyn_toothpick_base {parent=dyn_toothpick_base_master, loc=<0 0 0>, rot=<0 0 0 1>}
 dyn_plume_base {parent=dyn_toothpick_base, loc=<14.9771 0.1257 -0.0811>, rot=<-0.0327 -0.156 -0.0064 -0.9872>}
 dyn_top_plume_01 {parent=dyn_plume_base, loc=<5.285 0 1.3872>, rot=<0 -0.2591 0 -0.9658>}
 dyn_top_plume_02 {parent=dyn_top_plume_01, loc=<6.0001 0 0>, rot=<0 -0.1369 0 -0.9906>}
 dyn_top_plume_03 {parent=dyn_top_plume_02, loc=<3.7401 0 0>, rot=<0 -0.2194 0 -0.9756>}
 dyn_top_plume_04 {parent=dyn_top_plume_03, loc=<6.7145 0 0>, rot=<0 -0.2327 0 -0.9725>}
 dyn_bot_plume_01 {parent=dyn_plume_base, loc=<4.0972 0.0446 -0.569>, rot=<0.0268 -0.313 -0.0264 -0.949>}
 dyn_bot_plume_02 {parent=dyn_bot_plume_01, loc=<4.5526 0 0>, rot=<-0.0367 -0.1164 0.0131 -0.9924>}
 dyn_bot_plume_03 {parent=dyn_bot_plume_02, loc=<3.6686 0 0>, rot=<0.0016 -0.255 0 -0.9669>}
 dyn_bot_plume_04 {parent=dyn_bot_plume_03, loc=<5.3071 0 0>, rot=<-0.0294 -0.1444 -0.0136 -0.989>}
 dyn_tongue_master {parent=dyn_head, loc=<0.4818 5.4807 -1.5697>, rot=<-0.3618 -0.6187 -0.4218 -0.5554>}
 dyn_tongue_01 {parent=dyn_tongue_master, loc=<0 0 0>, rot=<0 0 0 1>}
 dyn_tongue_02 {parent=dyn_tongue_01, loc=<7.6841 0 0>, rot=<0.0158 -0.0662 0.0033 -0.9977>}
 dyn_tongue_03 {parent=dyn_tongue_02, loc=<5.6938 0 0>, rot=<-0.0201 -0.0128 -0.0207 -0.9995>}
 dyn_tongue_04 {parent=dyn_tongue_03, loc=<5.4072 0 0>, rot=<0.0085 -0.0156 -0.0003 -0.9998>}
 dyn_helmet_visor_master {parent=dyn_head, loc=<8.2747 -0.2753 0>, rot=<0.4979 0.5021 0.5021 0.4979>}
 dyn_helmet_visor {parent=dyn_helmet_visor_master, loc=<0 0 0>, rot=<0 0 0 1>}
 dyn_cheese_fr_r_master {parent=dyn_head, loc=<-4.2574 11.0634 7.8224>, rot=<0.0687 0.4809 0.8741 0.0008>}
 dyn_cheese_fr_r {parent=dyn_cheese_fr_r_master, loc=<0 0 0>, rot=<0 0 0 -1>}
 dyn_cheese_bk_l_master {parent=dyn_head, loc=<-3.1004 -9.7787 -7.151>, rot=<0.0696 -0.8816 0.4616 -0.0701>}
 dyn_cheese_bk_l {parent=dyn_cheese_bk_l_master, loc=<0 0 0>, rot=<0 0 0 1>}
 dyn_cheese_fr_l_master {parent=dyn_head, loc=<-3.4125 7.7607 -10.6847>, rot=<0.0216 -0.9451 -0.3173 -0.0756>}
 dyn_cheese_fr_l {parent=dyn_cheese_fr_l_master, loc=<0 0 0>, rot=<0 0 0 -1>}
 dyn_burger_top_master {parent=dyn_head, loc=<4.1603 1.0462 0>, rot=<-0.5677 -0.4216 -0.4216 -0.5677>}
 dyn_burger_top {parent=dyn_burger_top_master, loc=<0 0 0>, rot=<0 0 0 1>}
 dyn_bot_bun {parent=dyn_head, loc=<0 0 0>, rot=<0 0 0 1>}
]
