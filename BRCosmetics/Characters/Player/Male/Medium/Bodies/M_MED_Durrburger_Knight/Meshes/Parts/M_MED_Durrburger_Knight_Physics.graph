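# Physics asset for the Durrburger Knight character body, setting up secondary motion for its dangling tongue and three cheese pieces.
# Master bodies lock to dyn_head, which has free angular motion relative to spine_05, and anchor a four-link limited tongue chain and tightly limited cheese bodies.
# The tongue root swings on one axis only, while later links bend and twist.
# Larger capsules cover the upper spine, left clavicle and unconstrained left upper arm.
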PhysicsAsset /BRCosmetics/Characters/Player/Male/Medium/Bodies/M_MED_Durrburger_Knight/Meshes/Parts/M_MED_Durrburger_Knight_Physics
bodies {
 clavicle_l [{shapeType=Capsule, length=6.2411, radius=9.7041, center=<8.1566 -0.0002 -0.6758>, rotation=<rot 90 0 0>}]
 dyn_cheese_bk_l [{shapeType=Capsule, length=2, radius=2, center=<2.4093 0 0>, rotation=<rot 90 0 0>}]
 dyn_cheese_bk_l_master [{shapeType=Capsule, length=2, radius=2}]
 dyn_cheese_fr_l [{shapeType=Capsule, length=2, radius=2, center=<2.4093 0 0>, rotation=<rot 90 0 0>}]
 dyn_cheese_fr_l_master [{shapeType=Capsule, length=2, radius=2}]
 dyn_cheese_fr_r [{shapeType=Capsule, length=2, radius=2, center=<2.4093 0 0>, rotation=<rot 90 0 0>}]
 dyn_cheese_fr_r_master [{shapeType=Capsule, length=2, radius=2}]
 dyn_head [{shapeType=Capsule, length=2, radius=2}]
 dyn_tongue_01 [{shapeType=Capsule, length=2, radius=2, center=<3.0713 0 0>, rotation=<rot 90 0 0>}]
 dyn_tongue_02 [{shapeType=Capsule, length=2.7027, radius=4.2172, center=<3 0 2>, rotation=<rot 0 0 90>}]
 dyn_tongue_03 [{shapeType=Capsule, length=4.5001, radius=4.0954, center=<2.5 0 2>, rotation=<rot 0 0 90>}]
 dyn_tongue_04 [{shapeType=Capsule, length=3.5, radius=4.0954, center=<2.5 0 2>, rotation=<rot 0 0 90>}]
 dyn_tongue_master [{shapeType=Capsule, length=2, radius=2}]
 spine_05 [{shapeType=Capsule, length=6.242, radius=17.0022, rotation=<rot 90 7.125 -4.1169>}]
 upperarm_l [{shapeType=Capsule, length=7.2411, radius=8.7041, center=<2.65 -0.0002 1.9196>, rotation=<rot 90 0 0>}]
}
constraints [
 {
  bone1Name=dyn_tongue_01
  bone2Name=dyn_tongue_master
  swing1LimitDegrees=15
  swing2LimitDegrees=3
  swing2Motion=Locked
  twistLimitDegrees=3
  twistMotion=Locked
 }
 {
  bone1Name=dyn_tongue_master
  bone2Name=dyn_head
  swing1LimitDegrees=45
  swing1Motion=Locked
  swing2LimitDegrees=45
  swing2Motion=Locked
  twistLimitDegrees=45
  twistMotion=Locked
 }
 {bone1Name=dyn_tongue_02, bone2Name=dyn_tongue_01, swing1LimitDegrees=30, swing2LimitDegrees=30, twistLimitDegrees=10}
 {bone1Name=dyn_tongue_03, bone2Name=dyn_tongue_02, swing1LimitDegrees=30, swing2LimitDegrees=30, twistLimitDegrees=10}
 {bone1Name=dyn_tongue_04, bone2Name=dyn_tongue_03, swing1LimitDegrees=30, swing2LimitDegrees=30, twistLimitDegrees=10}
 {
  bone1Name=dyn_cheese_fr_r_master
  bone2Name=dyn_head
  swing1LimitDegrees=45
  swing1Motion=Locked
  swing2LimitDegrees=45
  swing2Motion=Locked
  twistLimitDegrees=45
  twistMotion=Locked
 }
 {bone1Name=dyn_cheese_fr_r, bone2Name=dyn_cheese_fr_r_master, swing1LimitDegrees=5, swing2LimitDegrees=8, twistLimitDegrees=5}
 {
  bone1Name=dyn_cheese_bk_l_master
  bone2Name=dyn_head
  swing1LimitDegrees=45
  swing1Motion=Locked
  swing2LimitDegrees=45
  swing2Motion=Locked
  twistLimitDegrees=45
  twistMotion=Locked
 }
 {bone1Name=dyn_cheese_bk_l, bone2Name=dyn_cheese_bk_l_master, swing1LimitDegrees=5, swing2LimitDegrees=8, twistLimitDegrees=5}
 {
  bone1Name=dyn_cheese_fr_l_master
  bone2Name=dyn_head
  swing1LimitDegrees=45
  swing1Motion=Locked
  swing2LimitDegrees=45
  swing2Motion=Locked
  twistLimitDegrees=45
  twistMotion=Locked
 }
 {bone1Name=dyn_cheese_fr_l, bone2Name=dyn_cheese_fr_l_master, swing1LimitDegrees=5, swing2LimitDegrees=8, twistLimitDegrees=5}
 {
  bone1Name=dyn_head
  bone2Name=spine_05
  swing1LimitDegrees=45
  swing1Motion=Free
  swing2LimitDegrees=45
  swing2Motion=Free
  twistLimitDegrees=45
  twistMotion=Free
 }
 {
  bone1Name=clavicle_l
  bone2Name=spine_05
  swing1LimitDegrees=45
  swing1Motion=Free
  swing2LimitDegrees=45
  swing2Motion=Free
  twistLimitDegrees=45
  twistMotion=Free
 }
]
