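# Physics asset for the visor part of the Durrburger Knight male character, covering the helmet visor and burger top bones.
# Matching large capsules represent the visor and burger top, with small capsules on their master bones and the head.
# Each piece is constrained to its master with hinge-like angle limits, wide only on swing 2.
# The visor master is angularly locked to the head, while the burger top master has no head constraint.
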PhysicsAsset /BRCosmetics/Characters/Player/Male/Medium/Bodies/M_MED_Durrburger_Knight/Meshes/Parts/M_MED_Durrburger_Knight_Physics_Visor
bodies {
 dyn_burger_top [{shapeType=Capsule, length=15.4647, radius=9.4869, center=<8.58 0.1293 2.424>, rotation=<rot 0 0.402 -88.75>}]
 dyn_burger_top_master [{shapeType=Capsule, length=2, radius=2}]
 dyn_head [{shapeType=Capsule, length=2, radius=2}]
 dyn_helmet_visor [{shapeType=Capsule, length=15.4647, radius=9.4869, center=<8.58 0.1293 2.424>, rotation=<rot 0 0.402 -88.75>}]
 dyn_helmet_visor_master [{shapeType=Capsule, length=2, radius=2}]
}
constraints [
 {
  bone1Name=dyn_helmet_visor_master
  bone2Name=dyn_head
  swing1LimitDegrees=45
  swing1Motion=Locked
  swing2LimitDegrees=45
  swing2Motion=Locked
  twistLimitDegrees=45
  twistMotion=Locked
 }
 {bone1Name=dyn_helmet_visor, bone2Name=dyn_helmet_visor_master, swing1LimitDegrees=1, swing2LimitDegrees=22, twistLimitDegrees=1}
 {bone1Name=dyn_burger_top, bone2Name=dyn_burger_top_master, swing1LimitDegrees=1, swing2LimitDegrees=20, twistLimitDegrees=1}
]
