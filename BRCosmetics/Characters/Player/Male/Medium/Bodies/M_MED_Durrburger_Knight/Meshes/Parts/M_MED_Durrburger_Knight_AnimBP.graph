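# Post-process style AnimBP for the Durrburger Knight body cosmetic, adding secondary motion to a pose copied from the attached parent mesh.
# The live chain offsets tongue bones, simulates two RigidBody assets and AnimDynamics toothpick and plumes, then applies damped head rotation.
# EventGraph derives correction alphas from bone deltas, weapon type, aiming and crouch.
# Bun and patty correctives, constraints and Control Rig are disconnected from Output Pose.
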AnimBlueprint /BRCosmetics/Characters/Player/Male/Medium/Bodies/M_MED_Durrburger_Knight/Meshes/Parts/M_MED_Durrburger_Knight_AnimBP
target_skeleton=&/M_MED_Durrburger_Knight_Skeleton
graphs [
 AnimGraph {nodes [
   "Constraint - dyn_patty_tweak (PPPPPPPPP)"!(
    ComponentPose
    Alpha=1
    ConstraintWeights_0=0
    ConstraintWeights_1:float=0.2883
    ConstraintWeights_2:float=0
    ConstraintWeights_3:float=0
    ConstraintWeights_4:float=0.2281
    ConstraintWeights_5:float=0.1934
    ConstraintWeights_6:float=0.0638
    ConstraintWeights_7:float=0.2258
    ConstraintWeights_8:float=0.0006
   ) -> (Pose) {
    alphaBoolBlend {blendOption=Linear}
    boneToModify {boneName=dyn_patty_tweak}
    constraintSetup [
     {targetBone {boneName=clavicle_l}}
     {targetBone {boneName=dyn_head}}
     {targetBone {boneName=dyn_cheese_bk_l}}
     {targetBone {boneName=dyn_cheese_fr_l}}
     {targetBone {boneName=spine_05}}
     {targetBone {boneName=dyn_bot_bun}}
     {targetBone {boneName=clavicle_r}}
     {targetBone {boneName=neck_01}}
     {targetBone {boneName=dyn_cheese_fr_r}}
    ]
    constraintWeights [1,1,1,1,1,1,1,1,1]
   }
   Control Rig() {
    alphaBoolBlend {blendOption=Linear}
    controlRigAssetReference {blueprintRigClass=&/M_MED_Durrburger_Knight_FaceAcc_CtrlRig.M_MED_Durrburger_Knight_FaceAcc_CtrlRig_C}
   }
   n1 = Get crouch_rifle_moving() -> (crouch_rifle_moving:double)
   "Transform (Modify) Bone - Bone: dyn_patty_tweak"(Alpha=@n1, Translation="2,2,-2", Rotation="0,0,0", Scale="1,1,1") {
    rotationMode=BMM_Additive
    scaleMode=BMM_Additive
    translationMode=BMM_Additive
    translationSpace=BCS_ParentBoneSpace
    alphaBoolBlend {blendOption=Linear}
    boneToModify {boneName=dyn_patty_tweak}
    translation=<2 2 -2>
   }
   n2 = "Constraint - dyn_bot_bun (P)"(Alpha=0.2, ConstraintWeights_0=1) {
    alphaBoolBlend {blendOption=Linear}
    boneToModify {boneName=dyn_bot_bun}
    constraintSetup [{targetBone {boneName=neck_01}}]
    constraintWeights [1]
   }
   n3 = "Transform (Modify) Bone - Bone: dyn_head"(ComponentPose=@n2, Alpha=Get rifle_adjustment() -> (rifle_adjustment:double), Translation="0.5,3.5,-1", Rotation="0,0,0", Scale="1,1,1") {
    rotationMode=BMM_Additive
    scaleMode=BMM_Additive
    translationMode=BMM_Additive
    alphaBoolBlend {blendOption=Linear}
    boneToModify {boneName=dyn_head}
    translation=<0.5 3.5 -1>
   }
   n4 = "Transform (Modify) Bone - Bone: dyn_bot_bun"(ComponentPose=@n3, Alpha=Get aim_adjustment_up() -> (aim_adjustment_up:double), Translation="-5,0,0", Rotation="0,0,0", Scale="1,1,1") {
    rotationMode=BMM_Additive
    scaleMode=BMM_Additive
    translationMode=BMM_Additive
    alphaBoolBlend {blendOption=Linear}
    boneToModify {boneName=dyn_bot_bun}
    translation {x=-5}
   }
   n5 = "Transform (Modify) Bone - Bone: dyn_bot_bun"(ComponentPose=@n4, Alpha=Get aim_up() -> (aim_up:double), Translation="0,-2,2", Rotation="25,30,23", Scale="1,1,1") {
    rotationMode=BMM_Additive
    scaleMode=BMM_Additive
    translationMode=BMM_Additive
    alphaBoolBlend {blendOption=Linear}
    boneToModify {boneName=dyn_bot_bun}
    rotation=<rot 25 30 23>
    translation {y=-2, z=2}
   }
   "Transform (Modify) Bone - Bone: dyn_bot_bun"(ComponentPose=@n5, Alpha=Get rifle_aim_up() -> (rifle_aim_up:double), Translation="-2,-5,2", Rotation="25,30,23", Scale="1,1,1") {
    rotationMode=BMM_Additive
    scaleMode=BMM_Additive
    translationMode=BMM_Additive
    alphaBoolBlend {blendOption=Linear}
    boneToModify {boneName=dyn_bot_bun}
    rotation=<rot 25 30 23>
    translation=<-2 -5 2>
   }
   "Transform (Modify) Bone - Bone: dyn_patty_tweak"(Alpha=Get crouch_adjustment() -> (crouch_adjustment:double), Translation="3,3,0", Rotation="0,0,0", Scale="1,1,1") {
    rotationMode=BMM_Additive
    scaleMode=BMM_Additive
    translationMode=BMM_Additive
    alphaBoolBlend {blendOption=Linear}
    boneToModify {boneName=dyn_patty_tweak}
    translation=<3 3>
   }
   n6 = "Transform (Modify) Bone - Bone: dyn_bot_bun"(Alpha=Get down_adjust() -> (down_adjust:double), Translation="1,2,0", Rotation="0,0,0", Scale="1,1,1") {
    rotationMode=BMM_Additive
    scaleMode=BMM_Additive
    translationMode=BMM_Additive
    alphaBoolBlend {blendOption=Linear}
    boneToModify {boneName=dyn_bot_bun}
    translation=<1 2>
   }
   n7 = "Transform (Modify) Bone - Bone: dyn_bot_bun"(ComponentPose=@n6, Alpha=Get pistol_aim() -> (pistol_aim:double), Translation="0,0,0", Rotation="0,0,0", Scale="1,1,1") {
    rotationMode=BMM_Additive
    scaleMode=BMM_Additive
    translationMode=BMM_Additive
    alphaBoolBlend {blendOption=Linear}
    boneToModify {boneName=dyn_bot_bun}
   }
   n8 = "Transform (Modify) Bone - Bone: dyn_bot_bun"(ComponentPose=@n7, Alpha=Get rifle_crouch_adjust() -> (rifle_crouch_adjust:double), Translation="-2,-1,0", Rotation="0,0,0", Scale="1,1,1") {
    rotationMode=BMM_Additive
    scaleMode=BMM_Additive
    translationMode=BMM_Additive
    alphaBoolBlend {blendOption=Linear}
    boneToModify {boneName=dyn_bot_bun}
    translation=<-2 -1>
   }
   n9 = "Transform (Modify) Bone - Bone: dyn_bot_bun"(ComponentPose=@n8, Alpha=Get rifle_up_adjust() -> (rifle_up_adjust:double), Translation="-3,-2,0", Rotation="0,0,0", Scale="1,1,1") {
    rotationMode=BMM_Additive
    scaleMode=BMM_Additive
    translationMode=BMM_Additive
    alphaBoolBlend {blendOption=Linear}
    boneToModify {boneName=dyn_bot_bun}
    translation=<-3 -2>
   }
   n10 = "Transform (Modify) Bone - Bone: dyn_bot_bun"(ComponentPose=@n9, Alpha=@n1, Translation="0,3,-3", Rotation="0,0,0", Scale="1,1,1") {
    rotationMode=BMM_Additive
    scaleMode=BMM_Additive
    translationMode=BMM_Additive
    translationSpace=BCS_WorldSpace
    alphaBoolBlend {blendOption=Linear}
    boneToModify {boneName=dyn_bot_bun}
    translation {y=3, z=-3}
   }
   "Transform (Modify) Bone - Bone: dyn_head"(ComponentPose=@n10, Alpha=@n1, Translation="1,1,-1", Rotation="0,0,0", Scale="1,1,1") {
    rotationMode=BMM_Additive
    scaleMode=BMM_Additive
    translationMode=BMM_Additive
    alphaBoolBlend {blendOption=Linear}
    boneToModify {boneName=dyn_head}
    translation=<1 1 -1>
   }
   n11 = "Transform (Modify) Bone - Bone: dyn_tongue_02"(
    ComponentPose=Local To Component(LocalPose=Copy Pose From Mesh() {bUseAttachedParent=true})
    Alpha=1
    Translation="0,0,0"
    Rotation="-40,0,0"
    Scale="1,1,1"
   ) {
    rotationMode=BMM_Additive
    rotationSpace=BCS_BoneSpace
    translationMode=BMM_Additive
    translationSpace=BCS_BoneSpace
    alphaBoolBlend {blendOption=Linear}
    boneToModify {boneName=dyn_tongue_02}
    rotation {pitch=-40}
   }
   n12 = "Transform (Modify) Bone - Bone: dyn_tongue_03"(ComponentPose=@n11, Alpha=1, Translation="0,0,0", Rotation="-25,0,0", Scale="1,1,1") {
    rotationMode=BMM_Additive
    rotationSpace=BCS_BoneSpace
    translationMode=BMM_Additive
    translationSpace=BCS_BoneSpace
    alphaBoolBlend {blendOption=Linear}
    boneToModify {boneName=dyn_tongue_03}
    rotation {pitch=-25}
   }
   n13 = RigidBody(ComponentPose=@n12) {
    bClampLinearTranslationLimitToRefPose=true
    bOverrideWorldGravity=true
    overridePhysicsAsset=&/M_MED_Durrburger_Knight_Physics
    simulationSpace=BaseBoneSpace
    alphaBoolBlend {blendOption=Linear}
    baseBoneRef {boneName=root}
    componentAppliedLinearAccClamp=<3000 3000 3000>
    componentLinearAccScale=<0.4 0.4 0.4>
    componentLinearVelScale=<0.4 0.4 0.4>
    overrideWorldGravity {z=-980}
   }
   n14 = RigidBody(ComponentPose=@n13) {
    bClampLinearTranslationLimitToRefPose=true
    bOverrideWorldGravity=true
    overridePhysicsAsset=&/M_MED_Durrburger_Knight_Physics_Visor
    simulationSpace=BaseBoneSpace
    alphaBoolBlend {blendOption=Linear}
    baseBoneRef {boneName=dyn_head}
    componentAppliedLinearAccClamp=<3000 3000 3000>
    componentLinearAccScale=<0.4 0.4 0.4>
    componentLinearVelScale=<0.4 0.4 0.4>
    overrideWorldGravity {z=-980}
   }
   n15 = Anim Dynamics - Bone: dyn_toothpick_base(ComponentPose=@n14, Alpha=1, ExternalForce="0,0,0") {
    angularDampingOverride=0.7
    angularSpringConstant=1000
    bGravityOverrideInSimSpace=true
    bUseGravityOverride=true
    simulationSpace=BoneRelative
    alphaBoolBlend {blendOption=Linear}
    boundBone {boneName=dyn_toothpick_base}
    componentAppliedLinearAccClamp=<1500 1500 1500>
    componentLinearAccScale=<0.4 0.4 0.4>
    componentLinearVelScale=<0.4 0.4 0.4>
    gravityOverride {z=-980}
    physicsBodyDefinitions [{
      sphereCollisionRadius=0
      boundBone {boneName=dyn_toothpick_base}
      boxExtents=<20 20 20>
      constraintSetup {angularLimitsMax=<2 2 2>, angularLimitsMin=<-2 -2 -2>, angularTarget=<1 0 0>}
      localJointOffset=<5 0 0>
     }]
    relativeSpaceBone {boneName=root}
   }
   n16 = Anim Dynamics - Bone: dyn_plume_base(ComponentPose=@n15, Alpha=1, ExternalForce="0,0,0") {
    angularDampingOverride=0.8
    angularSpringConstant=1000
    bGravityOverrideInSimSpace=true
    bUseGravityOverride=true
    simulationSpace=BoneRelative
    alphaBoolBlend {blendOption=Linear}
    boundBone {boneName=dyn_plume_base}
    componentAppliedLinearAccClamp=<1500 1500 1500>
    componentLinearAccScale=<0.4 0.4 0.4>
    componentLinearVelScale=<0.4 0.4 0.4>
    gravityOverride {z=-980}
    physicsBodyDefinitions [{
      sphereCollisionRadius=0
      boundBone {boneName=dyn_plume_base}
      boxExtents=<20 20 20>
      constraintSetup {angularLimitsMax=<3 10 3>, angularLimitsMin=<-3 -10 -3>, angularTarget=<1 0 0>}
      localJointOffset=<5 0 0>
     }]
    relativeSpaceBone {boneName=root}
   }
   n17 = Anim Dynamics - Bone: dyn_top_plume_01(ComponentPose=@n16, Alpha=1, ExternalForce="0,0,0") {
    angularDampingOverride=0.7
    angularSpringConstant=1000
    bGravityOverrideInSimSpace=true
    bUseGravityOverride=true
    alphaBoolBlend {blendOption=Linear}
    boundBone {boneName=dyn_top_plume_01}
    componentAppliedLinearAccClamp=<1500 1500 1500>
    componentLinearAccScale=<0.4 0.4 0.4>
    componentLinearVelScale=<0.4 0.4 0.4>
    gravityOverride {z=-980}
    physicsBodyDefinitions [{
      sphereCollisionRadius=0
      boundBone {boneName=dyn_top_plume_01}
      boxExtents=<20 20 20>
      constraintSetup {angularLimitsMax=<3 15 3>, angularLimitsMin=<-3 -15 -3>, angularTarget=<1 0 0>}
      localJointOffset=<5 0 0>
     }]
    relativeSpaceBone {boneName=root}
   }
   n18 = Anim Dynamics - Bone: dyn_top_plume_02(ComponentPose=@n17, Alpha=1, ExternalForce="0,0,0") {
    angularDampingOverride=0.7
    angularSpringConstant=1000
    bGravityOverrideInSimSpace=true
    bUseGravityOverride=true
    alphaBoolBlend {blendOption=Linear}
    boundBone {boneName=dyn_top_plume_02}
    componentAppliedLinearAccClamp=<1500 1500 1500>
    componentLinearAccScale=<0.4 0.4 0.4>
    componentLinearVelScale=<0.4 0.4 0.4>
    gravityOverride {z=-980}
    physicsBodyDefinitions [{
      sphereCollisionRadius=0
      boundBone {boneName=dyn_top_plume_02}
      boxExtents=<20 20 20>
      constraintSetup {angularLimitsMax=<3 15 3>, angularLimitsMin=<-3 -15 -3>, angularTarget=<1 0 0>}
      localJointOffset=<5 0 0>
     }]
    relativeSpaceBone {boneName=root}
   }
   n19 = Anim Dynamics - Bone: dyn_top_plume_03(ComponentPose=@n18, Alpha=1, ExternalForce="0,0,0") {
    angularDampingOverride=0.7
    angularSpringConstant=1000
    bGravityOverrideInSimSpace=true
    bUseGravityOverride=true
    alphaBoolBlend {blendOption=Linear}
    boundBone {boneName=dyn_top_plume_03}
    componentAppliedLinearAccClamp=<1500 1500 1500>
    componentLinearAccScale=<0.4 0.4 0.4>
    componentLinearVelScale=<0.4 0.4 0.4>
    gravityOverride {z=-980}
    physicsBodyDefinitions [{
      sphereCollisionRadius=0
      boundBone {boneName=dyn_top_plume_03}
      boxExtents=<20 20 20>
      constraintSetup {angularLimitsMax=<3 15 3>, angularLimitsMin=<-3 -15 -3>, angularTarget=<1 0 0>}
      localJointOffset=<5 0 0>
     }]
    relativeSpaceBone {boneName=root}
   }
   n20 = Anim Dynamics - Bone: dyn_top_plume_04(ComponentPose=@n19, Alpha=1, ExternalForce="0,0,0") {
    angularDampingOverride=0.7
    angularSpringConstant=1000
    bGravityOverrideInSimSpace=true
    bUseGravityOverride=true
    alphaBoolBlend {blendOption=Linear}
    boundBone {boneName=dyn_top_plume_04}
    componentAppliedLinearAccClamp=<1500 1500 1500>
    componentLinearAccScale=<0.4 0.4 0.4>
    componentLinearVelScale=<0.4 0.4 0.4>
    gravityOverride {z=-980}
    physicsBodyDefinitions [{
      sphereCollisionRadius=0
      boundBone {boneName=dyn_top_plume_04}
      boxExtents=<20 20 20>
      constraintSetup {angularLimitsMax=<3 15 3>, angularLimitsMin=<-3 -15 -3>, angularTarget=<1 0 0>}
      localJointOffset=<5 0 0>
     }]
    relativeSpaceBone {boneName=root}
   }
   n21 = Anim Dynamics - Bone: dyn_bot_plume_01(ComponentPose=@n20, Alpha=1, ExternalForce="0,0,0") {
    angularDampingOverride=0.7
    angularSpringConstant=1000
    bGravityOverrideInSimSpace=true
    bUseGravityOverride=true
    alphaBoolBlend {blendOption=Linear}
    boundBone {boneName=dyn_bot_plume_01}
    componentAppliedLinearAccClamp=<1500 1500 1500>
    componentLinearAccScale=<0.4 0.4 0.4>
    componentLinearVelScale=<0.4 0.4 0.4>
    gravityOverride {z=-980}
    physicsBodyDefinitions [{
      sphereCollisionRadius=0
      boundBone {boneName=dyn_bot_plume_01}
      boxExtents=<20 20 20>
      constraintSetup {angularLimitsMax=<3 15 3>, angularLimitsMin=<-3 -15 -3>, angularTarget=<1 0 0>}
      localJointOffset=<5 0 0>
     }]
    relativeSpaceBone {boneName=root}
   }
   n22 = Anim Dynamics - Bone: dyn_bot_plume_02(ComponentPose=@n21, Alpha=1, ExternalForce="0,0,0") {
    angularDampingOverride=0.7
    angularSpringConstant=1000
    bGravityOverrideInSimSpace=true
    bUseGravityOverride=true
    alphaBoolBlend {blendOption=Linear}
    boundBone {boneName=dyn_bot_plume_02}
    componentAppliedLinearAccClamp=<1500 1500 1500>
    componentLinearAccScale=<0.4 0.4 0.4>
    componentLinearVelScale=<0.4 0.4 0.4>
    gravityOverride {z=-980}
    physicsBodyDefinitions [{
      sphereCollisionRadius=0
      boundBone {boneName=dyn_bot_plume_02}
      boxExtents=<20 20 20>
      constraintSetup {angularLimitsMax=<3 15 3>, angularLimitsMin=<-3 -15 -3>, angularTarget=<1 0 0>}
      localJointOffset=<5 0 0>
     }]
    relativeSpaceBone {boneName=root}
   }
   n23 = Anim Dynamics - Bone: dyn_bot_plume_03(ComponentPose=@n22, Alpha=1, ExternalForce="0,0,0") {
    angularDampingOverride=0.7
    angularSpringConstant=1000
    bGravityOverrideInSimSpace=true
    bUseGravityOverride=true
    alphaBoolBlend {blendOption=Linear}
    boundBone {boneName=dyn_bot_plume_03}
    componentAppliedLinearAccClamp=<1500 1500 1500>
    componentLinearAccScale=<0.4 0.4 0.4>
    componentLinearVelScale=<0.4 0.4 0.4>
    gravityOverride {z=-980}
    physicsBodyDefinitions [{
      sphereCollisionRadius=0
      boundBone {boneName=dyn_bot_plume_03}
      boxExtents=<20 20 20>
      constraintSetup {angularLimitsMax=<3 15 3>, angularLimitsMin=<-3 -15 -3>, angularTarget=<1 0 0>}
      localJointOffset=<5 0 0>
     }]
    relativeSpaceBone {boneName=root}
   }
   n24 = Anim Dynamics - Bone: dyn_bot_plume_04(ComponentPose=@n23, Alpha=1, ExternalForce="0,0,0") {
    angularDampingOverride=0.7
    angularSpringConstant=1000
    bGravityOverrideInSimSpace=true
    bUseGravityOverride=true
    alphaBoolBlend {blendOption=Linear}
    boundBone {boneName=dyn_bot_plume_04}
    componentAppliedLinearAccClamp=<1500 1500 1500>
    componentLinearAccScale=<0.4 0.4 0.4>
    componentLinearVelScale=<0.4 0.4 0.4>
    gravityOverride {z=-980}
    physicsBodyDefinitions [{
      sphereCollisionRadius=0
      boundBone {boneName=dyn_bot_plume_04}
      boxExtents=<20 20 20>
      constraintSetup {angularLimitsMax=<3 15 3>, angularLimitsMin=<-3 -15 -3>, angularTarget=<1 0 0>}
      localJointOffset=<5 0 0>
     }]
    relativeSpaceBone {boneName=root}
   }
   n25 = "Transform (Modify) Bone - Bone: dyn_head"(ComponentPose=@n24, Alpha=1, Translation="0,0,0", Rotation=Get HeadRot() -> (HeadRot:Rotator), Scale="1,1,1") {rotationMode=BMM_Replace, rotationSpace=BCS_BoneSpace, alphaBoolBlend {blendOption=Linear}, boneToModify {boneName=dyn_head}}
   n26 = "Transform (Modify) Bone - Bone: dyn_head"(ComponentPose=@n25, Alpha=Get headRotBack() -> (headRotBack:double), Translation="0,-2,0", Rotation="0,0,0", Scale="1,1,1") {
    rotationMode=BMM_Additive
    scaleMode=BMM_Additive
    translationMode=BMM_Additive
    alphaBoolBlend {blendOption=Linear}
    boneToModify {boneName=dyn_head}
    translation {y=-2}
   }
   Output Pose(Result=Component To Local(ComponentPose=@n26)) {layerGroup=None}
  ]}
 EventGraph {nodes [
   Event BlueprintUpdateAnimation()
   > n1 = Sequence()
   n2 = GetOwnerSkeletalMeshComponent()
   n3 = BreakRotator(InRot=BreakTransform(InTransform=GetDeltaTransformFromRefPose(self=@n2, BoneName=head, BaseName=neck_02)).Rotation)
   > Set HeadRot(HeadRot:Rotator=MakeRotator(
     Roll=float * float(A=@n3, B=0.25) {is=title K2Node_PromotableOperator}
     Pitch=float * float(A=@n3.Pitch, B=0.5)
     Yaw=float * float(A=@n3.Yaw, B=0.5)
    )) -> (Output_Get:Rotator)
   n4 = BreakTransform(InTransform=GetDeltaTransformFromRefPose(self=GetOwnerSkeletalMeshComponent(), BoneName=head, BaseName=spine_05))
   > Set headRotBack(headRotBack:double=MapRangeClamped(Value=BreakRotator(InRot=@n4.Rotation).Yaw, InRangeA=0, InRangeB=-50)) -> (Output_Get:double)
   Set aim_up(execute=@n1.then_1, aim_up:double=MapRangeClamped(Value=0, InRangeA=95, InRangeB=120)) -> (Output_Get:double)
   n5 = GetOwnerSkeletalMeshComponent()
   n6 = BreakRotator(InRot=BreakTransform(InTransform=GetDeltaTransformFromRefPose(self=@n5, BoneName=upperarm_r, BaseName=clavicle_r)).Rotation)
   > Set aim_adjustment_up(aim_adjustment_up:double=float - float(A=MapRangeClamped(Value=@n6.Yaw, InRangeA=92, InRangeB=114), B=MapRangeClamped(Value=@n6.Yaw, InRangeA=114, InRangeB=120, OutRangeB=0.6)) {is=title K2Node_CallFunction}) -> (Output_Get:double)
   n7 = BreakTransform(InTransform=GetDeltaTransformFromRefPose(self=GetOwnerSkeletalMeshComponent(), BoneName=clavicle_l, BaseName=spine_05))
   > Set rifle_aim_up(rifle_aim_up:double=MapRangeClamped(Value=BreakRotator(InRot=@n7.Rotation), InRangeA=0, InRangeB=10)) -> (Output_Get:double)
   > n8 = Sequence!(execute) -> (then_0, then_1, then_2:Exec)
   > n9 = Cast To FortPlayerAnimInstance(Object=GetAnimInstance(self=GetOwnerSkeletalMeshComponent()))
   > Set rifle_adjustment(rifle_adjustment:double=0) -> (Output_Get:double)
   > Set pistol_out(pistol_out:double=0) -> (Output_Get:double)
   > Set rifle_out(rifle_out:bool=false) -> (Output_Get:bool)
   > n10 = Switch on EFortWeaponCoreAnimation!(execute, Selection:EFortWeaponCoreAnimation Enum=Get WeaponCoreAnim(self=@n9)) -> (
    Melee:Exec
    Pistol:Exec
    Shotgun:Exec
    PaperBlueprint:Exec
    Rifle:Exec
    MeleeOneHand:Exec
    MachinePistol:Exec
    RocketLauncher:Exec
    GrenadeLauncher:Exec
    GoingCommando:Exec
    AssaultRifle:Exec
    TacticalShotgun:Exec
    SniperRifle:Exec
    TrapPlacement:Exec
    ShoulderLauncher:Exec
    AbilityDecoTool:Exec
    Crossbow:Exec
    C4:Exec
    RemoteControl:Exec
    DualWield:Exec
    AR_BullPup:Exec
    AR_ForwardGrip:Exec
    MedPackPaddles:Exec
    SMG_P90:Exec
    AR_DrumGun:Exec
    Consumable_Small:Exec
    Consumable_Large:Exec
    Balloon:Exec
    MountedTurret:Exec
    CreativeTool:Exec
    ExplosiveBow:Exec
    AshtonIndigo:Exec
    AshtonChicago:Exec
    MeleeDualWield:Exec
    Unarmed:Exec
   )
   n11 = Branch(execute=[@n10.AR_DrumGun, @n10.SMG_P90], Condition=Get bIsTargeting())
   n12 = BreakTransform(InTransform=GetDeltaTransformFromRefPose(self=GetOwnerSkeletalMeshComponent(), BoneName=upperarm_r, BaseName=clavicle_r))
   Set rifle_adjustment(
    execute=[@n11.then, @n15.else, @n13.then]
    rifle_adjustment=MapRangeClamped(Value=BreakRotator(InRot=@n12.Rotation).Yaw, InRangeA=60, InRangeB=25)
   )
   > Set rifle_out(rifle_out=true)
   Set rifle_adjustment(execute=[@n11.else, @n15.then, @n14.then], rifle_adjustment=0)
   > Set rifle_out(rifle_out=true)
   n13 = Branch(execute=[@n10.AssaultRifle, @n10.Rifle, @n10.Shotgun, @n10.SniperRifle, @n10.TacticalShotgun], Condition=Get bIsTargeting())
   n14 = Branch(execute=@n13.else, Condition=Get bIsRelaxedLevel1() -> (bIsRelaxedLevel1:bool))
   n15 = Branch(execute=@n14.else, Condition=Get bIsRelaxedLevel2() -> (bIsRelaxedLevel2:bool))
   Set pistol_out(execute=@n10.Pistol, pistol_out=1)
   n16 = Branch(execute=@n8.then_1, Condition=Get bIsCrouching())
   > n17 = Set crouch_adjustment(crouch_adjustment:double=1) -> (Output_Get:double)
   n18 = GetOwnerSkeletalMeshComponent()
   n19 = BreakRotator(InRot=BreakTransform(InTransform=GetDeltaTransformFromRefPose(self=@n18, BoneName=upperarm_r, BaseName=clavicle_r)).Rotation)
   > Set down_adjust(down_adjust:double=float - float(A=@n17, B=MapRangeClamped(Value=@n19.Yaw, InRangeA=140, InRangeB=75))) -> (Output_Get:double)
   > Set rifle_crouch_adjust(rifle_crouch_adjust:double=0) -> (Output_Get:double)
   > Set rifle_up_adjust(rifle_up_adjust:double=0) -> (Output_Get:double)
   > Set pistol_aim(pistol_aim:double=0) -> (Output_Get:double)
   > Set crouch_rifle_moving(crouch_rifle_moving:double=0) -> (Output_Get:double)
   > n20 = Branch(Condition="float >= float"(A=Get pistol_out() -> (pistol_out:double), B=1) {is=title K2Node_CallFunction})
   n21 = BreakTransform(InTransform=GetDeltaTransformFromRefPose(self=GetOwnerSkeletalMeshComponent(), BoneName=upperarm_r, BaseName=clavicle_r))
   > Set pistol_aim(pistol_aim=MapRangeClamped(Value=BreakRotator(InRot=@n21.Rotation).Yaw, InRangeA=90, InRangeB=120))
   Set pistol_aim(execute=@n20.else, pistol_aim=0)
   > n22 = Branch(Condition=Get rifle_out() -> (rifle_out:bool))
   n23 = BreakTransform(InTransform=GetDeltaTransformFromRefPose(self=GetOwnerSkeletalMeshComponent(), BoneName=upperarm_r, BaseName=clavicle_r))
   > Set rifle_up_adjust(rifle_up_adjust=MapRangeClamped(Value=BreakRotator(InRot=@n23.Rotation).Yaw, InRangeA=40, InRangeB=60))
   > n24 = Branch(Condition=Get bIsTargeting())
   n25 = BreakTransform(InTransform=GetDeltaTransformFromRefPose(self=GetOwnerSkeletalMeshComponent(), BoneName=upperarm_r, BaseName=clavicle_r))
   > n26 = Set rifle_crouch_adjust(rifle_crouch_adjust=MapRangeClamped(Value=BreakRotator(InRot=@n25.Rotation).Yaw, InRangeA=70, InRangeB=50))
   n27 = Get bIsMoving2D()
   > n28 = Branch(Condition=@n27)
   n29 = BreakTransform(InTransform=GetDeltaTransformFromRefPose(self=GetOwnerSkeletalMeshComponent(), BoneName=head, BaseName=neck_02))
   > Set crouch_rifle_moving(crouch_rifle_moving=MapRangeClamped(Value=BreakRotator(InRot=@n29.Rotation).Yaw, InRangeA=-15, InRangeB=-10))
   Set crouch_rifle_moving(execute=@n28.else, crouch_rifle_moving=0)
   Set rifle_crouch_adjust(execute=[@n24.else, @n22.else], rifle_crouch_adjust=0)
   > Set rifle_up_adjust(rifle_up_adjust=0)
   Set crouch_adjustment(execute=@n16.else, crouch_adjustment=0)
   n30 = Get CurrentLOD()
   n31 = Set low_LOD(execute=@n8.then_2, low_LOD:double=SelectFloat(A=0, B=1, bPickA="integer > integer"(A=@n30, B=1) {is=title K2Node_CallFunction})) -> (Output_Get:double)
   PrintString!(execute, InString="To String (Float)"(InDouble=@n31), bPrintToScreen, bPrintToLog, TextColor, Duration, Key) -> (then)
   "To String (Integer)"(InInt=@n30)
   AND Boolean(A=@n27, B="float > float"(A=@n26, B=0) {is=title K2Node_CallFunction})
  ]}
]
variables [
 aim_adjustment_up {type=number}
 aim_up {type=number}
 crouch_adjustment {type=number}
 crouch_rifle_moving {type=number}
 down_adjust {type=number}
 HeadRot {type=Rotator}
 headRotBack {type=number}
 low_LOD {type=number}
 pistol_aim {type=number}
 pistol_out {type=number}
 rifle_adjustment {type=number}
 rifle_aim_up {type=number}
 rifle_crouch_adjust {type=number}
 rifle_out {type=boolean}
 rifle_up_adjust {type=number}
]
